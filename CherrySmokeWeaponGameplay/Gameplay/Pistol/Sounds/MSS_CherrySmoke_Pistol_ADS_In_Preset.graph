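# One-shot mono MetaSound source for the CherrySmoke pistol's aim-down-sights-in foley.
# On play it picks a random wave from the Sounds array, then rolls a random pitch shift within a spread either side of a base value.
# A mono wave player plays the result, scaled by MasterVolume, and signals finished when the wave ends.
# Sound class, attenuation and concurrency use the shared Fort weapon foley assets.
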MetaSoundSource /CherrySmokeWeaponGameplay/Gameplay/Pistol/Sounds/MSS_CherrySmoke_Pistol_ADS_In_Preset
inputs [
 MasterVolume {type=Float}
 Random.PitchShift.Base {type=Float}
 Random.PitchShift.Spread {type=Float}
 Sounds {type=WaveAsset:Array}
 UE.Source.OnPlay {type=Trigger}
]
interfaces [UE.OutputFormat.Mono:1.0, UE.Source.OneShot:1.0, UE.Source:1.1]
outputs [
 "UE.OutputFormat.Mono.Audio:0" {type=Audio}
 UE.Source.OneShot.OnFinished {type=Trigger}
]
pages [Default {
  input_defaults {
   MasterVolume=0.5
   Random.PitchShift.Base=0
   Random.PitchShift.Spread=0.5
   UE.Source.OnPlay=null
   Sounds [
    &/Pistol_CherrySmoke_ADS_In_01
    &/Pistol_CherrySmoke_ADS_In_02
   ]
  }
  nodes [
   ArrayRandomGet_0 = Array.Random Get:WaveAsset:Array(In Array=^Sounds, Next=^UE.Source.OnPlay) {also {defaults {Weights []}}}
   n1 = UE.RandomFloat(
    Max=UE.Add:Float(PrimaryOperand=^Random.PitchShift.Base, AdditionalOperands=^Random.PitchShift.Spread).Out
    Next=@ArrayRandomGet_0.On Next
    Min=UE.Subtract:Float(PrimaryOperand=^Random.PitchShift.Base, AdditionalOperands=^Random.PitchShift.Spread).Out
   )
   WavePlayer_0 = UE.Wave Player:Mono(Wave Asset=@ArrayRandomGet_0.Value, Play=@n1.On Next, Pitch Shift=@n1.Value)
   Multiply_0 = UE.Multiply:Audio by Float(AdditionalOperands=^MasterVolume, PrimaryOperand=@WavePlayer_0.Out Mono)
  ]
  links [{dst="@UE.OutputFormat.Mono.Audio:0", src=Multiply_0.Out}, {dst="@UE.Source.OneShot.OnFinished", src=WavePlayer_0.On Finished}]
 }]
properties {
 assetImportData=&self:AssetImportData
 attenuationSettings=&AS_Fort_WeaponFoley
 maxDistance=2250
 sampleRate=48000
 soundClassObject=&Fort_WeaponFoley
 concurrencySet [&CS_Fort_WeaponFoley]
}
